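# Animation Blueprint for the EagerTrapper offhand pickaxe that layers procedural secondary motion onto the incoming pose.
# A RigidBody node simulates the pickaxe's override physics asset in base-bone space of b_AR_Root, with clamped velocity and acceleration.
# Its result feeds four chained AnimDynamics nodes on the left and right ear bones, using angular springs, reduced gravity and slightly different tuning per side.
# The event graph holds no logic.
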AnimBlueprint /BRCosmetics/Weapons/FORT_Melee/Pickaxe_EagerTrapper_Offhand/Meshes/EagerTrapper_Offhand_Axe_AnimBP
target_skeleton=&/EagerTrapper_Offhand_Axe_Skeleton
graphs [
 AnimGraph {nodes [
   AnimGraphNode_SubInput_0 = Input Pose()
   n1 = RigidBody!(ComponentPose=Local To Component(LocalPose=@AnimGraphNode_SubInput_0), Alpha, OverrideWorldGravity:Vector="0,0,-980", ExternalForce) -> (Pose) {
    bOverrideWorldGravity=true
    overridePhysicsAsset=&/EagerTrapper_Offhand_Axe_Physics
    simulationSpace=BaseBoneSpace
    alphaBoolBlend {blendOption=Linear}
    baseBoneRef {boneName=b_AR_Root}
    componentAppliedLinearAccClamp=<3000 3000 3000>
    simSpaceSettings {
     maxAngularAcceleration=1000
     maxAngularVelocity=15
     maxLinearAcceleration=300
     maxLinearVelocity=700
     worldAlpha=1
     externalLinearDragV {y=2}
    }
   }
   n2 = Anim Dynamics - Bone: dyn_ear_l_02(ComponentPose=@n1, Alpha=1, ExternalForce="0,0,0") {
    angularDampingOverride=1
    angularSpringConstant=20000
    bAngularSpring=true
    bOverrideAngularDamping=true
    bUseGravityOverride=true
    gravityScale=0.35
    simulationSpace=BoneRelative
    alphaBoolBlend {blendOption=Linear}
    boundBone {boneName=dyn_ear_l_02}
    componentAppliedLinearAccClamp=<2000 2000 2000>
    componentLinearAccScale=<0.7 0.7 0.7>
    componentLinearVelScale=<0.7 0.7 0.7>
    gravityOverride {z=-980}
    physicsBodyDefinitions [{
      boundBone {boneName=dyn_ear_l_02}
      boxExtents=<40 40 40>
      constraintSetup {angularLimitsMax=<0 5 20>, angularLimitsMin=<0 -45 -20>, angularTarget=<1 0 0>}
      localJointOffset=<13 0 0>
     }]
    relativeSpaceBone {boneName=b_AR_Root}
   }
   n3 = Anim Dynamics - Bone: dyn_ear_l_01(ComponentPose=@n2, Alpha=1, ExternalForce="0,0,0") {
    angularDampingOverride=1
    angularSpringConstant=20000
    bAngularSpring=true
    bOverrideAngularDamping=true
    bUseGravityOverride=true
    gravityScale=0.35
    simulationSpace=BoneRelative
    alphaBoolBlend {blendOption=Linear}
    boundBone {boneName=dyn_ear_l_01}
    componentAppliedLinearAccClamp=<2000 2000 2000>
    componentLinearAccScale=<0.7 0.7 0.7>
    componentLinearVelScale=<0.7 0.7 0.7>
    gravityOverride {z=-980}
    physicsBodyDefinitions [{
      boundBone {boneName=dyn_ear_l_01}
      boxExtents=<40 40 40>
      constraintSetup {angularLimitsMax=<0 5 20>, angularLimitsMin=<0 -45 -20>, angularTarget=<1 0 0>}
      localJointOffset=<13 0 0>
     }]
    relativeSpaceBone {boneName=b_AR_Root}
   }
   n4 = Anim Dynamics - Bone: dyn_ear_r_02(ComponentPose=@n3, Alpha=1, ExternalForce="0,0,0") {
    angularDampingOverride=1
    angularSpringConstant=20000
    bAngularSpring=true
    bOverrideAngularDamping=true
    bUseGravityOverride=true
    gravityScale=0.35
    simulationSpace=BoneRelative
    alphaBoolBlend {blendOption=Linear}
    boundBone {boneName=dyn_ear_r_02}
    componentAppliedLinearAccClamp=<2000 2000 2000>
    componentLinearAccScale=<0.6 0.6 0.6>
    componentLinearVelScale=<0.6 0.6 0.6>
    gravityOverride {z=-980}
    physicsBodyDefinitions [{
      boundBone {boneName=dyn_ear_r_02}
      boxExtents=<40 40 40>
      constraintSetup {angularLimitsMax=<0 5 30>, angularLimitsMin=<0 -35 -30>, angularTarget=<1 0 0>}
      localJointOffset=<13 0 0>
     }]
    relativeSpaceBone {boneName=b_AR_Root}
   }
   n5 = Anim Dynamics - Bone: dyn_ear_r_01(ComponentPose=@n4, Alpha=1, ExternalForce="0,0,0") {
    angularDampingOverride=1
    angularSpringConstant=20000
    bAngularSpring=true
    bOverrideAngularDamping=true
    bUseGravityOverride=true
    gravityScale=0.35
    simulationSpace=BoneRelative
    alphaBoolBlend {blendOption=Linear}
    boundBone {boneName=dyn_ear_r_01}
    componentAppliedLinearAccClamp=<2000 2000 2000>
    componentLinearAccScale=<0.6 0.6 0.6>
    componentLinearVelScale=<0.6 0.6 0.6>
    gravityOverride {z=-980}
    physicsBodyDefinitions [{
      boundBone {boneName=dyn_ear_r_01}
      boxExtents=<40 40 40>
      constraintSetup {angularLimitsMax=<0 5 30>, angularLimitsMin=<0 -35 -30>, angularTarget=<1 0 0>}
      localJointOffset=<13 0 0>
     }]
    relativeSpaceBone {boneName=b_AR_Root}
   }
   Output Pose(Result=Component To Local(ComponentPose=@n5)) {layerGroup=None}
  ]}
 EventGraph {nodes [Event BlueprintUpdateAnimation()]}
]
variables []
